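# Physics asset for the Cyber Fu Blade pickaxe, a Fortnite melee cosmetic, providing collision bodies and joints for its dangling pendant.
# A long capsule covers the two-handed handle, with two similar five-link capsule chains, `dyn_pendant` and `_dyn_pendant`, hanging from it.
# Each link has limited swing and tight twist.
# The `dyn_` chain joins the handle directly, while the `_dyn_` chain goes through a small base body whose joint rotates freely.
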PhysicsAsset /BRCosmetics/Weapons/FORT_Melee/Pickaxe_Cyber_Fu_Blade/Meshes/Cyber_Fu_Blade_Physics
bodies {
 Melee_TwoHanded_Handle [{shapeType=Capsule, length=136.5491, radius=2.2379, center=<0.9746 -0.9979 28.2455>}]
 _dyn_pendant_01 [{shapeType=Capsule, length=3.46, radius=1.47, shapeName=_dyn_pendant_01_capsule, center=<3.31 0.0114 0.0261>, rotation=<rot 90 -0.6 -0.6>}]
 _dyn_pendant_02 [{shapeType=Capsule, length=3.46, radius=1.47, shapeName=_dyn_pendant_02_capsule, center=<3.31 0 0>, rotation=<rot 90 -0.6 -0.6>}]
 _dyn_pendant_03 [{shapeType=Capsule, length=2.35, radius=2.33, shapeName=_dyn_pendant_03_capsule, center=<2.68 0 0>, rotation=<rot 90 0 0>}]
 _dyn_pendant_04 [{shapeType=Capsule, length=2.45, radius=2.7866, shapeName=_dyn_pendant_04_capsule, center=<1.64 0 0>, rotation=<rot 90 0 0>}]
 _dyn_pendant_05 [{shapeType=Capsule, length=2, radius=3.27, shapeName=_dyn_pendant_05_capsule, center=<1.64 0 0>, rotation=<rot 90 0 0>}]
 _dyn_pendant_basa [{shapeType=Capsule, length=0.505, radius=0.505, shapeName=_dyn_pendant_basa_capsule, rotation=<rot 0 -90 0>}]
 dyn_pendant_01 [{shapeType=Capsule, length=3.4629, radius=1.4779, center=<3.3159 0 -0.0005>, rotation=<rot 90 -0.6328 -0.6326>}]
 dyn_pendant_02 [{shapeType=Capsule, length=3.4629, radius=1.4779, center=<3.3159 0 -0.0005>, rotation=<rot 90 -0.6328 -0.6326>}]
 dyn_pendant_03 [{shapeType=Capsule, length=2.3512, radius=2.3342, center=<2.688 0 0.0003>, rotation=<rot 90 0.0066 0.0066>}]
 dyn_pendant_04 [{shapeType=Capsule, length=2.4516, radius=2.7573, center=<1.6425 0 0.0002>, rotation=<rot 90 0.0264 0.0264>}]
 dyn_pendant_05 [{shapeType=Capsule, length=1.9352, radius=3.2737, center=<1.6425 0 0.0002>, rotation=<rot 90 0.0264 0.0264>}]
}
constraints [
 {bone1Name=dyn_pendant_01, bone2Name=Melee_TwoHanded_Handle, swing1LimitDegrees=30, swing2LimitDegrees=30, twistLimitDegrees=5}
 {bone1Name=dyn_pendant_02, bone2Name=dyn_pendant_01, swing1LimitDegrees=30, swing2LimitDegrees=30, twistLimitDegrees=5}
 {bone1Name=dyn_pendant_03, bone2Name=dyn_pendant_02, swing1LimitDegrees=30, swing2LimitDegrees=30, twistLimitDegrees=5}
 {bone1Name=dyn_pendant_04, bone2Name=dyn_pendant_03, swing1LimitDegrees=30, swing2LimitDegrees=30, twistLimitDegrees=5}
 {bone1Name=dyn_pendant_05, bone2Name=dyn_pendant_04, swing1LimitDegrees=30, swing2LimitDegrees=30, twistLimitDegrees=5}
 {bone1Name=_dyn_pendant_05, bone2Name=_dyn_pendant_04, swing1LimitDegrees=30, swing2LimitDegrees=30, twistLimitDegrees=5}
 {
  bone1Name=_dyn_pendant_basa
  bone2Name=Melee_TwoHanded_Handle
  swing1LimitDegrees=45
  swing1Motion=Free
  swing2LimitDegrees=45
  swing2Motion=Free
  twistLimitDegrees=45
  twistMotion=Free
 }
 {bone1Name=_dyn_pendant_01, bone2Name=_dyn_pendant_basa, swing1LimitDegrees=30, swing2LimitDegrees=30, twistLimitDegrees=5}
 {bone1Name=_dyn_pendant_02, bone2Name=_dyn_pendant_01, swing1LimitDegrees=30, swing2LimitDegrees=30, twistLimitDegrees=5}
 {bone1Name=_dyn_pendant_03, bone2Name=_dyn_pendant_02, swing1LimitDegrees=30, swing2LimitDegrees=30, twistLimitDegrees=5}
 {bone1Name=_dyn_pendant_04, bone2Name=_dyn_pendant_03, swing1LimitDegrees=30, swing2LimitDegrees=30, twistLimitDegrees=5}
]
