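# Animation Blueprint for the Cyber Fu Blade pickaxe that adds secondary physics motion, chiefly to its dangling pendant bones.
# The input pose passes through a RigidBody node using the blade's physics asset, then five chained AnimDynamics nodes, one per pendant bone, simulated relative to the handle.
# A Trail controller for the pendant chain sits unconnected.
# The update event sets RBAN-off and LOD-2+ flags, which no visible connection reads.
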AnimBlueprint /BRCosmetics/Weapons/FORT_Melee/Pickaxe_Cyber_Fu_Blade/Meshes/Cyber_Fu_Blade_AnimBlueprint
target_skeleton=&/Cyber_Fu_Blade_Skeleton
graphs [
 AnimGraph {nodes [
   AnimGraphNode_SubInput_0 = Input Pose()
   n1 = RigidBody!(ComponentPose=Local To Component(LocalPose=@AnimGraphNode_SubInput_0), bAlphaBoolEnabled:bool, ExternalForce) -> (Pose) {
    alphaInputType=Bool
    bOverrideWorldGravity=true
    overridePhysicsAsset=&/Cyber_Fu_Blade_Physics
    simulationSpace=BaseBoneSpace
    alphaBoolBlend {blendOption=Linear}
    baseBoneRef {boneName=Melee_TwoHanded_Handle}
    overrideWorldGravity {z=-930}
    simSpaceSettings {maxAngularAcceleration=500, maxAngularVelocity=3, maxLinearAcceleration=400, maxLinearVelocity=250, velocityScaleZ=0.7, worldAlpha=1}
   }
   n2 = Anim Dynamics - Bone: _dyn_pendant_01!(ComponentPose=@n1, bAlphaBoolEnabled:bool=True, ExternalForce="0,0,0") -> (Pose) {
    alphaInputType=Bool
    angularDampingOverride=1
    bOverrideAngularDamping=true
    bUseGravityOverride=true
    simulationSpace=BoneRelative
    alphaBoolBlend {blendOption=Linear}
    boundBone {boneName=_dyn_pendant_01}
    componentLinearAccScale=<1 1 1>
    componentLinearVelScale=<1 1 1>
    gravityOverride {z=-980}
    physicsBodyDefinitions [{
      boundBone {boneName=_dyn_pendant_01}
      boxExtents=<12 12 12>
      constraintSetup {angularLimitsMax=<5 30 30>, angularLimitsMin=<-5 -30 0>, angularTarget=<1 0 0>}
      localJointOffset=<2 0 0>
     }]
    relativeSpaceBone {boneName=Melee_TwoHanded_Handle}
   }
   n3 = Anim Dynamics - Bone: _dyn_pendant_02!(ComponentPose=@n2, bAlphaBoolEnabled:bool=True, ExternalForce="0,0,0") -> (Pose) {
    alphaInputType=Bool
    angularDampingOverride=1
    bOverrideAngularDamping=true
    bUseGravityOverride=true
    simulationSpace=BoneRelative
    alphaBoolBlend {blendOption=Linear}
    boundBone {boneName=_dyn_pendant_02}
    componentLinearAccScale=<1 1 1>
    componentLinearVelScale=<1 1 1>
    gravityOverride {z=-980}
    physicsBodyDefinitions [{
      boundBone {boneName=_dyn_pendant_02}
      boxExtents=<12 12 12>
      constraintSetup {angularLimitsMax=<5 30 30>, angularLimitsMin=<-5 -30 -10>, angularTarget=<1 0 0>}
      localJointOffset=<2 0 0>
     }]
    relativeSpaceBone {boneName=Melee_TwoHanded_Handle}
   }
   n4 = Anim Dynamics - Bone: _dyn_pendant_03!(ComponentPose=@n3, bAlphaBoolEnabled:bool=True, ExternalForce="0,0,0") -> (Pose) {
    alphaInputType=Bool
    angularDampingOverride=1
    bOverrideAngularDamping=true
    bUseGravityOverride=true
    simulationSpace=BoneRelative
    alphaBoolBlend {blendOption=Linear}
    boundBone {boneName=_dyn_pendant_03}
    componentLinearAccScale=<1 1 1>
    componentLinearVelScale=<1 1 1>
    gravityOverride {z=-980}
    physicsBodyDefinitions [{
      boundBone {boneName=_dyn_pendant_03}
      boxExtents=<12 12 12>
      constraintSetup {angularLimitsMax=<5 30 30>, angularLimitsMin=<-5 -30 -30>, angularTarget=<1 0 0>}
      localJointOffset=<2 0 0>
     }]
    relativeSpaceBone {boneName=Melee_TwoHanded_Handle}
   }
   n5 = Anim Dynamics - Bone: _dyn_pendant_04!(ComponentPose=@n4, bAlphaBoolEnabled:bool=True, ExternalForce="0,0,0") -> (Pose) {
    alphaInputType=Bool
    angularDampingOverride=1
    bOverrideAngularDamping=true
    bUseGravityOverride=true
    simulationSpace=BoneRelative
    alphaBoolBlend {blendOption=Linear}
    boundBone {boneName=_dyn_pendant_04}
    componentLinearAccScale=<1 1 1>
    componentLinearVelScale=<1 1 1>
    gravityOverride {z=-980}
    physicsBodyDefinitions [{
      boundBone {boneName=_dyn_pendant_04}
      boxExtents=<12 12 12>
      constraintSetup {angularLimitsMax=<5 30 30>, angularLimitsMin=<-5 -30 -30>, angularTarget=<1 0 0>}
      localJointOffset=<2 0 0>
     }]
    relativeSpaceBone {boneName=Melee_TwoHanded_Handle}
   }
   n6 = Anim Dynamics - Bone: _dyn_pendant_05!(ComponentPose=@n5, bAlphaBoolEnabled:bool=True, ExternalForce="0,0,0") -> (Pose) {
    alphaInputType=Bool
    angularDampingOverride=1
    bOverrideAngularDamping=true
    bUseGravityOverride=true
    simulationSpace=BoneRelative
    alphaBoolBlend {blendOption=Linear}
    boundBone {boneName=_dyn_pendant_05}
    componentLinearAccScale=<1 1 1>
    componentLinearVelScale=<1 1 1>
    gravityOverride {z=-980}
    physicsBodyDefinitions [{
      boundBone {boneName=_dyn_pendant_05}
      boxExtents=<12 12 12>
      constraintSetup {angularLimitsMax=<5 30 30>, angularLimitsMin=<-5 -30 -30>, angularTarget=<1 0 0>}
      localJointOffset=<2 0 0>
     }]
    relativeSpaceBone {boneName=Melee_TwoHanded_Handle}
   }
   Output Pose(Result=Component To Local(ComponentPose=@n6)) {layerGroup=None}
   Trail controller - Bone: _dyn_pendant_05!(ComponentPose, bAlphaBoolEnabled:bool=True, RelaxationSpeedScale:float=1) -> (Pose) {
    alphaInputType=Bool
    bActorSpaceFakeVel=true
    bLimitRotation=true
    bShowTrailLocation=true
    bUsePlanarLimit=true
    chainLength=5
    lastBoneRotationAnimAlphaBlend=1
    maxDeltaTime=0.0333
    baseJoint {boneName=_dyn_pendant_basa}
    fakeVelocity {y=150, z=-150}
    relaxationSpeedScaleInputProcessor {bInterpResult=true, interpSpeedDecreasing=25, interpSpeedIncreasing=25}
    rotationLimits [
     {limitMax=<5 20 20>, limitMin=<-5 -20 -20>}
     {limitMax=<5 20 20>, limitMin=<-5 -20 -20>}
     {limitMax=<5 5 20>, limitMin=<-5 -5 -20>}
     {limitMax=<5 5 20>, limitMin=<-5 -5 -20>}
     {limitMax=<5 5 20>, limitMin=<-5 -5 -20>}
    ]
    rotationOffsets [<0 0 0>, <0 0 0>, <0 0 0>, <0 0 0>, <0 0 0>]
    trailBone {boneName=_dyn_pendant_05}
   }
  ]}
 EventGraph {nodes [
   Event BlueprintUpdateAnimation()
   > Set RBAN off(RBAN off:bool=NOT Boolean(A=Get bRigidBodyEnabled())) -> (Output_Get:bool)
   > Set Lod2(Lod2:bool="integer >= integer"(A=Get CurrentLOD(), B=2)) -> (Output_Get:bool)
   TryGetPawnOwner()
  ]}
]
variables [
 Lod2 {type=boolean}
 RBAN off {type=boolean}
]
