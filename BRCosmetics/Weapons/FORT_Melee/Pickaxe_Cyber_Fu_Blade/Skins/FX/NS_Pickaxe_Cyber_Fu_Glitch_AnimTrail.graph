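# Niagara swing trail for the glitch skin of the Cyber Fu Blade pickaxe, built from three ribbon emitters.
# Each uses the FN AnimTrail spawn, initialize and update modules to create short-lived ribbons that fade by alpha curve.
# Trail0 flickers through multicolor curve keys, Trail1 adds a dark layer, and thinner, longer-lived Trail2 shifts from bright cyan to blue.
# Time-of-day scalars scale brightness, plus opacity on Trail0 and Trail2.
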NiagaraSystem /BRCosmetics/Weapons/FORT_Melee/Pickaxe_Cyber_Fu_Blade/Skins/FX/NS_Pickaxe_Cyber_Fu_Glitch_AnimTrail
dependencies {
 usedDataInterfaces [&NiagaraDataInterfaceCurve, &NiagaraDataInterfaceColorCurve, &NiagaraDataInterfaceSocketReader]
 usedDynamicInputs []
 usedModules [
  &self:SystemSpawnScript.SystemScriptSource.SystemScriptGraph.NiagaraNodeAssignment_1.SetVariables_#1
  &SystemState
  &self:SystemSpawnScript.SystemScriptSource.SystemScriptGraph.NiagaraNodeAssignment_0.SetVariables_#2
 ]
}
emitters {
 Trail0 {
  emitterUpdateScript [
   &EmitterState()
   &FN_SpawnAnimTrail()
  ]
  particleSpawnScript [
   &FN_InitializeAnimTrailBoneIndecies()
   &FN_InitializeAnimTrail(AnimTrailWidth=0.8, Color=<rgba 1 1 1 0.5>, Lifetime=0.2)
  ]
  particleUpdateScript [
   &ScaleColor(
    Scale Alpha=&Multiply_Float(
     A=&FloatFromCurve(FloatCurve=DataInterface() {propertyValues {Curve {keys [
          {interpMode=RCIM_Cubic, leaveTangent=-2.3646, tangentMode=RCTM_Break, time=0, value=1}
          <at 0.6689 0.0233>
          {arriveTangent=-0.2139, interpMode=RCIM_Cubic, tangentMode=RCTM_Break, time=1, value=0}
         ]}}})
     B=&FNTimeOfDayScalar(Night Scale=1, Morning Scale=0.5, Evening Scale=0.5)
    )
    Scale RGB=&Multiply_VectorByFloat(Vector=<1 1 1>, Float=$System.TOD)
    ScaleA=true
    ScaleRGB=true
   ) {module=ScaleColor002}
   &FN_UpdateAnimTrail()
   &ScaleColor(
    Scale RGB=&Multiply_VectorByFloat(
     Vector=&MakeVectorFromLinearColorRGB(Color=&ColorFromCurve(ColorCurve=DataInterface() {propertyValues {
         AlphaCurve {keys []}
         BlueCurve {keys [<at 0 1 RCIM_Cubic>,<at 0.1234 0.4>,<at 0.3046 0.7976>,<at 0.4937 0.4555>,<at 0.6249 0.4>,<at 0.7311 1>,<at 0.8327 0.4675>]}
         GreenCurve {keys [<at 0 0.6129 RCIM_Cubic>,<at 0.1234 0.4064>,<at 0.3046 1>,<at 0.4937 0.4>,<at 0.6249 1>,<at 0.7311 0.4598>,<at 0.8327 0.4>]}
         RedCurve {keys [<at 0 0.3 RCIM_Cubic>,<at 0.1234 1>,<at 0.3046 0>,<at 0.4937 1>,<at 0.6249 0.79>,<at 0.7311 0>,<at 0.8327 1>]}
        }}))
     Float=1
    )
    ScaleA=true
    ScaleRGB=true
   )
  ]
  rendererClasses [&NiagaraRibbonRendererProperties]
  renderers [{rendererClass=&NiagaraRibbonRendererProperties}]
 }
 Trail1 {
  emitterUpdateScript [
   &EmitterState()
   &FN_SpawnAnimTrail()
  ]
  particleSpawnScript [
   &FN_InitializeAnimTrailBoneIndecies()
   &FN_InitializeAnimTrail(AnimTrailWidth=0.8, Color=<rgba 0.05 0.05 0.05 0.5>, Lifetime=0.2)
  ]
  particleUpdateScript [
   &ScaleColor(
    Scale Alpha=&FloatFromCurve(FloatCurve=DataInterface() {propertyValues {Curve {keys [
         {interpMode=RCIM_Cubic, leaveTangent=-2.3646, tangentMode=RCTM_Break, time=0, value=1}
         <at 0.6689 0.0233>
         {arriveTangent=-0.2139, interpMode=RCIM_Cubic, tangentMode=RCTM_Break, time=1, value=0}
        ]}}})
    Scale RGB=&Multiply_VectorByFloat(Vector=<1 1 1>, Float=$System.TOD)
    ScaleA=true
    ScaleRGB=true
   ) {module=ScaleColor002}
   &FN_UpdateAnimTrail()
   &ScaleColor(Scale RGB=&Multiply_VectorByFloat(Vector=<1 1 1>, Float=1), ScaleA=true, ScaleRGB=true)
  ]
  rendererClasses [&NiagaraRibbonRendererProperties]
  renderers [{rendererClass=&NiagaraRibbonRendererProperties}]
 }
 Trail2 {
  emitterUpdateScript [
   &EmitterState()
   &FN_SpawnAnimTrail()
  ]
  particleSpawnScript [
   &FN_InitializeAnimTrailBoneIndecies()
   &FN_InitializeAnimTrail(AnimTrailWidth=0.3, Lifetime=0.3)
  ]
  particleUpdateScript [
   &FN_UpdateAnimTrail()
   &ScaleColor(
    Scale Alpha=&Multiply_Float(
     A=&FloatFromCurve(FloatCurve=DataInterface() {propertyValues {Curve {keys [
          {interpMode=RCIM_Cubic, leaveTangent=-2.3646, tangentMode=RCTM_Break, time=0, value=1}
          <at 0.5 0.0233>
          {arriveTangent=-0.2139, interpMode=RCIM_Cubic, tangentMode=RCTM_Break, time=1, value=0}
         ]}}})
     B=$System.TOD
    )
    Scale RGB=&Multiply_VectorByFloat(
     Vector=&MakeVectorFromLinearColorRGB(Color=&ColorFromCurve(ColorCurve=DataInterface() {propertyValues {
         AlphaCurve {keys []}
         BlueCurve {keys [<at 0 3 RCIM_Cubic>,<at 0.2204 1 RCIM_Cubic>]}
         GreenCurve {keys [<at 0 2.3044 RCIM_Cubic>,<at 0.2204 0.1672 RCIM_Cubic>]}
         RedCurve {keys [<at 0 0 RCIM_Cubic>,<at 0.2204 0 RCIM_Cubic>]}
        }}))
     Float=$System.TOD
    )
    ScaleA=true
    ScaleRGB=true
   )
  ]
  rendererClasses [&NiagaraRibbonRendererProperties]
  renderers [{rendererClass=&NiagaraRibbonRendererProperties}]
 }
}
summary {
 systemName=NS_Pickaxe_Cyber_Fu_Glitch_AnimTrail
 emitters [
  {emitterName=Trail0, rendererClasses [&NiagaraRibbonRendererProperties]}
  {emitterName=Trail1, rendererClasses [&NiagaraRibbonRendererProperties]}
  {emitterName=Trail2, rendererClasses [&NiagaraRibbonRendererProperties]}
 ]
 userVariables [
  User.bIsFrontEndPreview {type=NiagaraBool}
  User.RibbonIDOffset {type=NiagaraInt32}
 ]
}
